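# Physics asset for the DuneBriefMerry male medium player body, providing body colliders, by its name for the cape simulation.
# One capsule per bone covers pelvis, spine, clavicles, arms, hands, legs and feet, with near-zero placeholder-sized capsules on the neck and none on the head.
# Left and right shapes match except the thighs, where the right capsule is longer and wider.
# Parent-child constraints leave swing and twist free, so their angle limits are inactive.
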PhysicsAsset /BRCosmetics/Characters/Player/Male/Medium/Bodies/M_MED_DuneBriefMerry/Meshes/M_MED_DuneBriefMerry_CapeColliders
bodies {
 calf_l [{shapeType=Capsule, length=31.185, radius=15.1339, shapeName=calf_l_capsule, center=<-25.2261 0 0>, rotation=<rot 90 -173.4436 0>}]
 calf_r [{shapeType=Capsule, length=31.185, radius=15.1339, shapeName=calf_r_capsule, center=<25.2261 0 0>, rotation=<rot 90 6.5564 0>}]
 clavicle_l [{shapeType=Capsule, length=9.8404, radius=5.3006, shapeName=clavicle_l_capsule, center=<7.7727 0 0>, rotation=<rot 90 0 0>}]
 clavicle_r [{shapeType=Capsule, length=9.8404, radius=5.3006, shapeName=clavicle_r_capsule, center=<-7.7727 0 0>, rotation=<rot 90 180 0>}]
 foot_l [{
   shapeType=Capsule
   length=22.6585
   radius=13.4048
   shapeName=foot_l_capsule
   center=<-1.8447 7.9751 0.7629>
   rotation=<rot 0 177.7889 93.8714>
  }]
 foot_r [{
   shapeType=Capsule
   length=22.6585
   radius=13.4048
   shapeName=foot_r_capsule
   center=<1.8447 -7.9751 -0.7629>
   rotation=<rot 0 177.7889 -86.1286>
  }]
 hand_l [{shapeType=Capsule, length=7.4164, radius=9.079, shapeName=hand_l_capsule, center=<7.4694 0 0>, rotation=<rot 90 0 0>}]
 hand_r [{shapeType=Capsule, length=7.4164, radius=9.079, shapeName=hand_r_capsule, center=<-7.4694 0 0>, rotation=<rot 90 180 0>}]
 lowerarm_l [{
   shapeType=Capsule
   length=15.822
   radius=10.0743
   shapeName=lowerarm_l_capsule
   center=<9.4106 0.2889 -0.7905>
   rotation=<rot 0 -90.4018 -89.2456>
  }]
 lowerarm_r [{
   shapeType=Capsule
   length=15.822
   radius=10.0743
   shapeName=lowerarm_r_capsule
   center=<-9.4106 -0.2889 0.7905>
   rotation=<rot 0 -90.4018 90.7544>
  }]
 neck_01 [{shapeType=Capsule, length=0.505, radius=0.505, shapeName=neck_01_capsule, rotation=<rot 0 -90 0>}]
 neck_02 [{shapeType=Capsule, length=0.505, radius=0.505, shapeName=neck_02_capsule, rotation=<rot 0 -90 0>}]
 pelvis [{shapeType=Capsule, length=5.5, radius=11.6793, shapeName=pelvis_capsule}]
 spine_01 [{shapeType=Capsule, length=7.005, radius=10.505, shapeName=spine_01_capsule}]
 spine_02 [{shapeType=Capsule, length=5.3799, radius=9.0959, shapeName=spine_02_capsule}]
 spine_03 [{shapeType=Capsule, length=5, radius=8.5748, shapeName=spine_03_capsule}]
 spine_04 [{shapeType=Capsule, length=8.5844, radius=10.0709, shapeName=spine_04_capsule}]
 spine_05 [{shapeType=Capsule, length=11.1069, radius=10.0615, shapeName=spine_05_capsule}]
 thigh_l [{
   shapeType=Capsule
   length=21.6737
   radius=7.9028
   shapeName=thigh_l_capsule
   center=<-18.8522 1.7975 -0.5258>
   rotation=<rot 0 -73.785 -87.1805>
  }]
 thigh_r [{shapeType=Capsule, length=27.6737, radius=9.4016, shapeName=thigh_r_capsule, center=<20.2449 0 0>, rotation=<rot 90 0 0>}]
 upperarm_l [{
   shapeType=Capsule
   length=23.239
   radius=5.2283
   shapeName=upperarm_l_capsule
   center=<13.0639 0 -0.7188>
   rotation=<rot 87.3192 -179.9983 -179.9984>
  }]
 upperarm_r [{
   shapeType=Capsule
   length=23.239
   radius=5.2283
   shapeName=upperarm_r_capsule
   center=<-13.0639 0 0.7188>
   rotation=<rot 87.3192 -179.9983 0.0016>
  }]
}
constraints [
 {
  bone1Name=spine_01
  bone2Name=pelvis
  swing1LimitDegrees=45
  swing1Motion=Free
  swing2LimitDegrees=45
  swing2Motion=Free
  twistLimitDegrees=45
  twistMotion=Free
 }
 {
  bone1Name=spine_02
  bone2Name=spine_01
  swing1LimitDegrees=45
  swing1Motion=Free
  swing2LimitDegrees=45
  swing2Motion=Free
  twistLimitDegrees=45
  twistMotion=Free
 }
 {
  bone1Name=spine_03
  bone2Name=spine_02
  swing1LimitDegrees=45
  swing1Motion=Free
  swing2LimitDegrees=45
  swing2Motion=Free
  twistLimitDegrees=45
  twistMotion=Free
 }
 {
  bone1Name=spine_04
  bone2Name=spine_03
  swing1LimitDegrees=45
  swing1Motion=Free
  swing2LimitDegrees=45
  swing2Motion=Free
  twistLimitDegrees=45
  twistMotion=Free
 }
 {
  bone1Name=spine_05
  bone2Name=spine_04
  swing1LimitDegrees=45
  swing1Motion=Free
  swing2LimitDegrees=45
  swing2Motion=Free
  twistLimitDegrees=45
  twistMotion=Free
 }
 {
  bone1Name=clavicle_l
  bone2Name=spine_05
  swing1LimitDegrees=45
  swing1Motion=Free
  swing2LimitDegrees=45
  swing2Motion=Free
  twistLimitDegrees=45
  twistMotion=Free
 }
 {
  bone1Name=upperarm_l
  bone2Name=clavicle_l
  swing1LimitDegrees=45
  swing1Motion=Free
  swing2LimitDegrees=45
  swing2Motion=Free
  twistLimitDegrees=45
  twistMotion=Free
 }
 {
  bone1Name=lowerarm_l
  bone2Name=upperarm_l
  swing1LimitDegrees=45
  swing1Motion=Free
  swing2LimitDegrees=45
  swing2Motion=Free
  twistLimitDegrees=45
  twistMotion=Free
 }
 {
  bone1Name=hand_l
  bone2Name=lowerarm_l
  swing1LimitDegrees=45
  swing1Motion=Free
  swing2LimitDegrees=45
  swing2Motion=Free
  twistLimitDegrees=45
  twistMotion=Free
 }
 {
  bone1Name=clavicle_r
  bone2Name=spine_05
  swing1LimitDegrees=45
  swing1Motion=Free
  swing2LimitDegrees=45
  swing2Motion=Free
  twistLimitDegrees=45
  twistMotion=Free
 }
 {
  bone1Name=upperarm_r
  bone2Name=clavicle_r
  swing1LimitDegrees=45
  swing1Motion=Free
  swing2LimitDegrees=45
  swing2Motion=Free
  twistLimitDegrees=45
  twistMotion=Free
 }
 {
  bone1Name=lowerarm_r
  bone2Name=upperarm_r
  swing1LimitDegrees=45
  swing1Motion=Free
  swing2LimitDegrees=45
  swing2Motion=Free
  twistLimitDegrees=45
  twistMotion=Free
 }
 {
  bone1Name=hand_r
  bone2Name=lowerarm_r
  swing1LimitDegrees=45
  swing1Motion=Free
  swing2LimitDegrees=45
  swing2Motion=Free
  twistLimitDegrees=45
  twistMotion=Free
 }
 {
  bone1Name=neck_01
  bone2Name=spine_05
  swing1LimitDegrees=45
  swing1Motion=Free
  swing2LimitDegrees=45
  swing2Motion=Free
  twistLimitDegrees=45
  twistMotion=Free
 }
 {
  bone1Name=neck_02
  bone2Name=neck_01
  swing1LimitDegrees=45
  swing1Motion=Free
  swing2LimitDegrees=45
  swing2Motion=Free
  twistLimitDegrees=45
  twistMotion=Free
 }
 {
  bone1Name=thigh_l
  bone2Name=pelvis
  swing1LimitDegrees=45
  swing1Motion=Free
  swing2LimitDegrees=45
  swing2Motion=Free
  twistLimitDegrees=45
  twistMotion=Free
 }
 {
  bone1Name=calf_l
  bone2Name=thigh_l
  swing1LimitDegrees=45
  swing1Motion=Free
  swing2LimitDegrees=45
  swing2Motion=Free
  twistLimitDegrees=45
  twistMotion=Free
 }
 {
  bone1Name=foot_l
  bone2Name=calf_l
  swing1LimitDegrees=45
  swing1Motion=Free
  swing2LimitDegrees=45
  swing2Motion=Free
  twistLimitDegrees=45
  twistMotion=Free
 }
 {
  bone1Name=thigh_r
  bone2Name=pelvis
  swing1LimitDegrees=45
  swing1Motion=Free
  swing2LimitDegrees=45
  swing2Motion=Free
  twistLimitDegrees=45
  twistMotion=Free
 }
 {
  bone1Name=calf_r
  bone2Name=thigh_r
  swing1LimitDegrees=45
  swing1Motion=Free
  swing2LimitDegrees=45
  swing2Motion=Free
  twistLimitDegrees=45
  twistMotion=Free
 }
 {
  bone1Name=foot_r
  bone2Name=calf_r
  swing1LimitDegrees=45
  swing1Motion=Free
  swing2LimitDegrees=45
  swing2Motion=Free
  twistLimitDegrees=45
  twistMotion=Free
 }
]
